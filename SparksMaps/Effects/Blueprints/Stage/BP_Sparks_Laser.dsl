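; BP_Sparks_Laser is a stage-show laser effect built on a Niagara component.
; The construction script assigns the laser system and pushes static color, color-override, jam and length-multiplier variables into it.
; Activate events and their Mojo function twins toggle the laser at a Strength threshold, while impulses reset and show it.
; Other Mojo hooks are empty stubs, and a LaserDelay event runs a timed on/off pulse.
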(event Custom|LaserDelay
  (Components|Activation|Activate (Variables|Default|GetLaser) false)
  (Utilities|FlowControl|Delay 5.0)
  (Components|Activation|Deactivate (Variables|Default|GetLaser)))

(event EventStageShowActivate
  (params (Strength :type Float%20%28single-precision%29) (MusicTimeMeasure :type Float%20%28single-precision%29))
  (if (> Strength 0.5)
    (Components|Activation|Activate (Variables|Default|GetLaser) false)
    (Rendering|SetVisibility (Variables|Default|GetLaser) true false)
    (else
      (Components|Activation|Deactivate (Variables|Default|GetLaser))
      (Rendering|SetVisibility (Variables|Default|GetLaser) false false))))

(event EventStageShowImpulse
  (params (MusicTimeMeasure :type Float%20%28single-precision%29))
  (bind _laser (Variables|Default|GetLaser))
  (Components|Activation|Activate _laser true)
  (Rendering|SetVisibility _laser true false))

(fn MojoStageShowActivate
  (params (Strength :type Float%20%28double-precision%29) (MusicTimeMeasure :type Float%20%28double-precision%29))
  (if (> Strength 0.5)
    (Components|Activation|Activate (Variables|Default|GetLaser) false)
    (Rendering|SetVisibility (Variables|Default|GetLaser) true false)
    (else
      (Components|Activation|Deactivate (Variables|Default|GetLaser))
      (Rendering|SetVisibility (Variables|Default|GetLaser) false false)))
  (return false))

(fn MojoStageShowActivateBeat
  (params (Active :type Float%20%28double-precision%29) (MusicTimeMeasure :type Float%20%28double-precision%29))
  (return false))

(fn MojoStageShowColorShift
  (params (Primary :type Linear%20Color%20Structure) (Secondary :type Linear%20Color%20Structure) (Tertiary :type Linear%20Color%20Structure) (AltPrimary :type Linear%20Color%20Structure) (AltSecondary :type Linear%20Color%20Structure) (AltTertiary :type Linear%20Color%20Structure) (MusicTimeMeasure :type Float%20%28double-precision%29))
  (return false))

(fn MojoStageShowCustomMode
  (params (ModeIndex :type Integer) (MusicTimeMeasure :type Float%20%28double-precision%29))
  (return false))

(fn MojoStageShowImpulse
  (params (MusicTimeMeasure :type Float%20%28double-precision%29))
  (bind _laser (Variables|Default|GetLaser))
  (Components|Activation|Activate _laser true)
  (Rendering|SetVisibility _laser true false)
  (return false))

(fn MojoStageShowOnStop
  (return false))

(fn MojoStageShowUpdateNiagaraSystem
  (params (Niagara%20System :type Niagara%20System%20Object%20Reference))
  (return false))

(fn UpdateNiagaraStaticColor
  (Niagara|SetNiagaraVariable(LinearColor) (Variables|Default|GetLaser) "Color" (Variables|Default|GetStatic%20Color))
  (Niagara|SetNiagaraVariable(Float) (Variables|Default|GetLaser) "UseMaterialColorOverride" (Math|Conversions|ToFloat(Boolean) (not (Variables|Default|GetUseStaticColor))))
  (Niagara|SetNiagaraVariable(Bool) (Variables|Default|GetLaser) "Jam?" (Variables|Default|GetIsJam))
  (Niagara|SetNiagaraVariable(Float) (Variables|Default|GetLaser) "LaserLengthMult" (Variables|Default|GetLaserLenghtMultiplier)))

(fn ConstructionScript
  (Niagara|SetNiagaraSystemAsset (Variables|Default|GetLaser) (Variables|Default|GetNS_Laser) true)
  (CallFunction|UpdateNiagaraStaticColor))
